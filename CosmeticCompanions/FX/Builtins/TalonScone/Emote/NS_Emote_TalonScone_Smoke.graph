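# Niagara smoke effect for the TalonScone cosmetic companion emote, combining a rising ribbon trail with steam sprites.
# SmokeTrail001 emits a narrowing ribbon whose alpha combines a lifetime curve scaled by time of day with a System.Age envelope that fades it in, then out.
# SteamSprites spawns growing, drifting puffs from a small ring during roughly the first second, using randomized SubUV frames and age-driven dynamic material parameters.
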NiagaraSystem /CosmeticCompanions/FX/Builtins/TalonScone/Emote/NS_Emote_TalonScone_Smoke
dependencies {usedDataInterfaces [&NiagaraDataInterfaceCurve], usedDynamicInputs [], usedModules [&SystemState]}
emitters {
 SmokeTrail001 {
  emitterSpawnScript [/SetVariables(Emitter.SystemRand01:NiagaraFloat=&RandomRangeFloat(Minimum=0, Maximum=1)) {isSetParameters=true}]
  emitterUpdateScript [
   &EmitterState()
   &SpawnRate(SpawnRate=30, Use Spawn Probability=false)
  ]
  particleSpawnScript [
   &InitializeParticle(
    Color=<rgba 0.2396 0.2396 0.2396 2.2>
    Lifetime=0.6
    Position Offset Coordinate Space=::Local
    Ribbon Facing Mode=::NewEnumerator1
    Ribbon Facing Vector:Vector3f=<1 0 0>
    Ribbon Width=25
    Ribbon Width Mode=::NewEnumerator1
    UsePositionOffset=true
    Write Material Random=true
   )
   &AddVelocity(Velocity=<0 0 70>)
  ]
  particleUpdateScript [
   &ParticleState()
   &ScaleColor(
    Scale Alpha=&Multiply_Float(
     A=&FloatFromCurve(
      FloatCurve=DataInterface() {propertyValues {Curve {keys [
          {arriveTangent=1.7132, interpMode=RCIM_Cubic, leaveTangent=1.7132, tangentMode=RCTM_User, time=0, value=0.4855}
          {arriveTangent=-0.6069, interpMode=RCIM_Cubic, leaveTangent=-0.6069, time=0.5, value=1}
          <at 0.8 0 RCIM_Cubic>
         ]}}}
      Scale Curve=&FNTimeOfDayScalar(Night Scale=1, Morning Scale=0.3, Evening Scale=0.5)
     )
     B=1
    )
    Scale RGB=&MakeVectorFromLinearColorRGB(Color=/Niagara/DynamicInputs/TypeConversions/MakeLinearColorFromFloat(Value:NiagaraFloat=&FloatFromCurve(FloatCurve=DataInterface() {propertyValues {Curve {keys [
           {arriveTangent=0.5628, interpMode=RCIM_Cubic, leaveTangent=0.5628, tangentMode=RCTM_User, time=0, value=0.6519}
           {interpMode=RCIM_Cubic, leaveTangent=2.2656, time=1, value=1}
          ]}}})))
    ScaleA=true
    ScaleRGB=true
   )
   &ScaleRibbonWidth(Ribbon Width Scale=&FloatFromCurve(FloatCurve=DataInterface() {propertyValues {Curve {keys [{tangentMode=RCTM_User, time=0, value=1}, {tangentMode=RCTM_User, time=1, value=0.1678}]}}}))
   &ScaleVelocity(Velocity Scale=&VectorFromFloat(Value=&FloatFromCurve(
      FloatCurve=DataInterface() {propertyValues {Curve {keys [{arriveTangent=0.5171, interpMode=RCIM_Cubic, leaveTangent=0.5171, tangentMode=RCTM_User, time=0, value=0.35}, <at 1.9761 1>]}}}
      CurveIndex=$System.Age
     )))
   &SolveForcesAndVelocity(Clamp Velocity=false, Limit Acceleration=false)
   &ScaleColor(
    Scale Alpha=&FloatFromCurve(
     FloatCurve=DataInterface() {propertyValues {Curve {keys [
         <at 0.1742 0 RCIM_Cubic>
         {arriveTangent=0.1099, interpMode=RCIM_Cubic, leaveTangent=0.1099, tangentMode=RCTM_User, time=0.6975, value=1}
         {arriveTangent=0.0988, interpMode=RCIM_Cubic, leaveTangent=0.0988, tangentMode=RCTM_User, time=3.4, value=1}
         <at 3.75 0 RCIM_Cubic>
        ]}}}
     CurveIndex=$System.Age
     Scale Curve=0.1
    )
    ScaleA=true
    ScaleRGB=true
   ) {module=ScaleColor003}
  ]
  rendererClasses [&NiagaraRibbonRendererProperties]
  renderers [{rendererClass=&NiagaraRibbonRendererProperties}]
 }
 SteamSprites {
  emitterUpdateScript [
   &EmitterState(Life Cycle Mode=::NewEnumerator1, Loop Behavior=::NewEnumerator1, Loop Duration Mode=::NewEnumerator1, UseLoopDelay=false)
   &SpawnRate(
    SpawnRate=&FloatFromCurve(FloatCurve=DataInterface() {propertyValues {Curve {keys [<at 0.9 1 RCIM_Cubic>,<at 1 0>]}}}, CurveIndex=$System.Age, Scale Curve=9)
    Use Spawn Probability=false
   )
  ]
  particleSpawnScript [
   &InitializeParticle(
    Color=<rgba 0.3125 0.3125 0.3125 0.484>
    Lifetime Max=1
    Lifetime Mode=::NewEnumerator1
    Sprite Rotation Angle=&RandomRangeFloat(Minimum=-0.05, Maximum=0.3)
    Sprite Rotation Mode=::NewEnumerator1
    Sprite Size Mode=::NewEnumerator1
    Uniform Sprite Size Max=30
    Uniform Sprite Size Min=40
    UsePositionOffset=true
    Write Material Random=true
   )
   &ShapeLocation(MaskSpawnGroup=false, Override Seed=false, Ring Radius=3, Shape Primitive=::NewEnumerator4, UseOwnerScale=true)
   &AddVelocity(Velocity=&RandomRangeVector(Minimum=<-2 -2 15>, Maximum=<2 2 40>, Evaluation Type=::NewEnumerator1))
  ]
  particleUpdateScript [
   &ParticleState()
   &ScaleSpriteSize(Scale Factor=&Vector2DFromFloat(Value=&FloatFromCurve(
      FloatCurve=DataInterface() {propertyValues {Curve {keys [
          {arriveTangent=1.8751, interpMode=RCIM_Cubic, leaveTangent=1.8751, tangentMode=RCTM_User, time=0, value=0.25}
          {arriveTangent=0.72, interpMode=RCIM_Cubic, leaveTangent=0.72, tangentMode=RCTM_User, time=1, value=1.25}
         ]}}}
      Scale Curve=0.9
     )))
   &SubUVAnimation(
    Sub UV Play Rate:NiagaraFloat=&RandomRangeFloat(Minimum=0.03, Maximum=0.05)
    SubUV Animation Mode:ENiagara_SubUVLookupMode=::NewEnumerator3
    SubUV Start Time Offset:NiagaraFloat=&RandomRangeFloat(Minimum=0, Maximum=63)
   )
   &DynamicMaterialParameters(
    Index 0 Param 1=&FloatFromCurve(FloatCurve=DataInterface() {propertyValues {Curve {keys [<at 0 -0.25>,<at 1 -1>]}}})
    Index 0 Param 4=&OneMinusFloat(Float=$Particles.NormalizedAge)
    Module.Param0WriteEnabled=true
    Module.Param1WriteEnabled=false
    Module.Param2WriteEnabled=false
    Module.Param3WriteEnabled=true
   )
   &Drag(Drag=0.5, Use Linear Drag=true, Use Rotational Drag=false)
   &SolveForcesAndVelocity(Clamp Velocity=false, Limit Acceleration=false)
   &ScaleColor(
    Scale Alpha=&FloatFromCurve(FloatCurve=DataInterface() {propertyValues {Curve {keys [<at 0 1>, <at 0.3 1>, {arriveTangent=-1.4286, interpMode=RCIM_Cubic, leaveTangent=-1.4286, time=0.5264, value=1}, <at 1 0>]}}})
    ScaleA=true
    ScaleRGB=true
   )
  ]
  rendererClasses [&NiagaraSpriteRendererProperties]
  renderers [{rendererClass=&NiagaraSpriteRendererProperties}]
 }
}
summary {
 systemName=NS_Emote_TalonScone_Smoke
 emitters [
  {emitterName=SmokeTrail001, rendererClasses [&NiagaraRibbonRendererProperties]}
  {emitterName=SteamSprites, rendererClasses [&NiagaraSpriteRendererProperties]}
 ]
 userVariables [
  User.bIsFrontEnd {type=NiagaraBool}
  User.bIsFrontEndPreview {type=NiagaraBool}
  User.RibbonIDOffset {type=NiagaraInt32}
 ]
}
